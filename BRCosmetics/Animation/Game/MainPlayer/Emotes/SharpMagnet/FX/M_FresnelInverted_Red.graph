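# Additive unlit FX material for Niagara mesh particles in the SharpMagnet emote, producing a Fresnel-based glow.
# MF_Fresnel_FX, driven by an exponent parameter, is multiplied by particle color RGB and fed into MF_FX_EmissiveAdjustments.
# Despite the Red in its name, the tint comes from particle color; a Constant4Vector node is left unconnected.
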Material /BRCosmetics/Animation/Game/MainPlayer/Emotes/SharpMagnet/FX/M_FresnelInverted_Red
graph {
 properties {
  bUsedWithNiagaraMeshParticles=true
  blendMode=BLEND_Additive
  pixelDepthOffsetMode=PDOM_Legacy
  shadingModel=MSM_Unlit
  usedShadingModels=Unlit
 }
 nodes [
  Constant4Vector()
  n1 = ScalarParameter() {defaultValue=2, parameterName="Exponent (S)"}
  MaterialFunctionCall(Emissive=Multiply(
    A=MaterialFunctionCall(Exponent=@n1) {materialFunction=&MF_Fresnel_FX}.Fresnel
    B=ParticleColor().RGB
   )) {materialFunction=&MF_FX_EmissiveAdjustments}
 ]
}
