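# Niagara effect for the SharpMagnet emote that shows a small red ball as a mesh particle.
# A single burst emitter spawns the mesh with a camera offset and a fixed lifetime.
# Its color follows a black-to-red gradient indexed through a float curve by User.NormalizedNotifyProgress, so the emote's notify drives the glow rise and fade.
# A second color scale adjusts brightness by time of day.
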NiagaraSystem /BRCosmetics/Animation/Game/MainPlayer/Emotes/SharpMagnet/FX/NS_Emote_SharpMagnet_RedBall
dependencies {
 usedDataInterfaces [&NiagaraDataInterfaceColorCurve, &NiagaraDataInterfaceCurve]
 usedDynamicInputs []
 usedModules [&SystemState, &FN_Integrator]
}
emitters {SimpleSpriteBurst {
  emitterUpdateScript [
   &EmitterState(Life Cycle Mode=::NewEnumerator1, Loop Behavior=::NewEnumerator1, Loop Duration=2, UseLoopDelay=false)
   &SpawnBurst_Instantaneous(Use Spawn Probability=false, UseLoopCountLimit=false)
  ]
  particleSpawnScript [
   &InitializeParticle(
    Lifetime=4
    Mesh Scale Mode=::NewEnumerator3
    Mesh Uniform Scale=0.065
    Sprite Size Mode=::NewEnumerator3
    Uniform Sprite Size=50
    UsePositionOffset=false
    Write Material Random=true
   )
   &CameraOffset()
  ]
  particleUpdateScript [
   &ParticleState()
   &ScaleColor(
    Curve Index=&FloatFromCurve(
     FloatCurve=DataInterface() {propertyValues {Curve {keys [
         {arriveTangent=9.9532, interpMode=RCIM_Cubic, leaveTangent=9.9532, tangentMode=RCTM_User, time=0, value=0}
         {arriveTangent=2.9821, leaveTangent=-1.2308, tangentMode=RCTM_User, time=0.2, value=1}
         {interpMode=RCIM_Cubic, tangentMode=RCTM_User, time=0.75, value=0.8}
         {arriveTangent=-0.0405, interpMode=RCIM_Cubic, leaveTangent=-0.0405, tangentMode=RCTM_User, time=1, value=0}
        ]}}}
     CurveIndex=$User.NormalizedNotifyProgress
    )
    Linear Color Curve=DataInterface() {propertyValues {
      AlphaCurve {keys [<at 0 1>,<at 0.2455 1>,<at 0.7964 1>,<at 0.9722 1>]}
      BlueCurve {keys [<at 0.0083 0>,<at 0.657 0>,<at 0.9755 0>]}
      GreenCurve {keys [<at 0.0083 0>,<at 0.657 0>,<at 0.9755 0.1167>]}
      RedCurve {keys [<at 0.0083 0>,<at 0.657 1>,<at 0.9755 1>]}
     }}
    Scale Mode=::NewEnumerator2
   )
   &SolveForcesAndVelocity(Clamp Velocity=false, Limit Acceleration=false)
   &ScaleColor(
    Scale RGB=&VectorFromFloat(Value=&FNTimeOfDayScalar(Night Scale=1, Morning Scale=1.5, Day Scale=2, Evening Scale=1))
    ScaleA=true
    ScaleRGB=true
   ) {module=ScaleColor001}
  ]
  rendererClasses [&NiagaraMeshRendererProperties]
  renderers [{rendererClass=&NiagaraMeshRendererProperties, rendererIndex=0}]
 }}
summary {
 systemName=NS_Emote_SharpMagnet_RedBall
 emitters [{emitterName=SimpleSpriteBurst, rendererClasses [&NiagaraMeshRendererProperties]}]
 userVariables [
  User.bIsFrontEnd {type=NiagaraBool}
  User.bIsFrontEndPreview {type=NiagaraBool}
  User.bIsFullyDeployed {type=NiagaraBool}
  User.NormalizedNotifyProgress {type=NiagaraFloat, defaultValue=0.4, description=""}
 ]
}
