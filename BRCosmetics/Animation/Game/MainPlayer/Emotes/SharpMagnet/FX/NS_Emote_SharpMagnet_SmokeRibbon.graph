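# Niagara system for the SharpMagnet player emote that draws a faint smoke ribbon from one emitter with two ribbon renderers.
# User.SmokeHeight scales spawn rate and speed and shortens lifetime; User.NormalizedNotifyProgress stops spawning past midway and fades alpha.
# Progress also strengthens horizontal-only panning curl noise and later adds a -X push transformed to world space.
# Width and alpha taper by particle age and execution index, with alpha also scaled by time of day.
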NiagaraSystem /BRCosmetics/Animation/Game/MainPlayer/Emotes/SharpMagnet/FX/NS_Emote_SharpMagnet_SmokeRibbon
dependencies {
 usedDataInterfaces [&NiagaraDataInterfaceCurve]
 usedDynamicInputs []
 usedModules [&SystemState, &FN_Integrator]
}
emitters {SmokeRibbon {
  emitterUpdateScript [
   &EmitterState()
   &SpawnRate(
    SpawnRate=&Multiply_Float(
     A=&Multiply_Float(A=50, B=$User.SmokeHeight)
     B=&FloatFromCurve(FloatCurve=DataInterface() {propertyValues {Curve {keys [<at 0.5 1>,<at 0.6 0>]}}}, CurveIndex=$User.NormalizedNotifyProgress)
    )
    Use Spawn Probability=false
   )
  ]
  particleSpawnScript [
   &InitializeParticle(
    Color=<rgba 1 1 1 0.05>
    Lifetime=&Divide_Float(A=0.8, B=$User.SmokeHeight)
    Position Offset=<0 0 -5>
    Ribbon Width Mode=::NewEnumerator1
    UsePositionOffset=true
    Write Material Random=true
   )
   &AddVelocity(Rotation Coordinate Space=::World, Velocity Speed Scale=$User.SmokeHeight)
   &InheritVelocity(Inherited Velocity Amount Scale=&VectorFromFloat(Value=0.2), Speed Limit=true, UseSpeedThreshold=false)
  ]
  particleUpdateScript [
   &ParticleState()
   /Niagara/Modules/Update/Forces/CurlNoiseForce(
    Noise Frequency:NiagaraFloat=&Multiply_Float(A=0.0001, B=1)
    Noise Strength=&FloatFromCurve(
     FloatCurve=DataInterface() {propertyValues {Curve {keys [{arriveTangent=592.1932, interpMode=RCIM_Cubic, leaveTangent=592.1932, tangentMode=RCTM_User, time=0, value=100}, <at 1 300>]}}}
     Scale Curve=&Add_Float(
      A=&Multiply_Float(A=0.7, B=$User.SmokeHeight)
      B=&FloatFromCurve(
       FloatCurve=DataInterface() {propertyValues {Curve {keys [{interpMode=RCIM_Cubic, tangentMode=RCTM_User, time=0.45, value=0}, {interpMode=RCIM_Cubic, tangentMode=RCTM_User, time=1, value=1}]}}}
       CurveIndex=$User.NormalizedNotifyProgress
       Scale Curve=10
      )
     )
    )
    Pan Noise=true
    Pan Noise Field=<0 0 2>
    Random Seed:NiagaraInt32=&RandomRangeInt(Minimum=0, Maximum=100)
   ) {enabled=false}
   /Niagara/Modules/Update/Forces/V2/CurlNoiseForce(
    Noise Frequency=200
    Noise Strength=&FloatFromCurve(
     FloatCurve=DataInterface() {propertyValues {Curve {keys [{interpMode=RCIM_Cubic, tangentMode=RCTM_User, time=0, value=0.1}, {interpMode=RCIM_Cubic, tangentMode=RCTM_User, time=1, value=1}]}}}
     CurveIndex=$User.NormalizedNotifyProgress
     Scale Curve=800
    )
    Pan Noise=true
    Pan Noise Field=<0 0 1>
   ) {module=CurlNoiseForce001}
   /SetVariables(Transient.PhysicsForce:Vector3f=&Add_Vector(
     A=&Multiply_Vector(A=$Output.CurlNoiseForce001.CurlNoiseForce, B=<1 1 0>)
     B=&Multiply_VectorByFloat(
      Vector=&TransformVector(Vector=<-1 0 0>, Destination Space=::World)
      Float=&FloatFromCurve(
       FloatCurve=DataInterface() {propertyValues {Curve {keys [
           {arriveTangent=5.1592, interpMode=RCIM_Cubic, leaveTangent=5.1592, tangentMode=RCTM_User, time=0.45, value=0}
           {interpMode=RCIM_Cubic, tangentMode=RCTM_User, time=1, value=1}
          ]}}}
       CurveIndex=$User.NormalizedNotifyProgress
       Scale Curve=800
      )
     )
    )) {isSetParameters=true}
   &ScaleRibbonWidth(Ribbon Width Scale=&Multiply_Float(
     A=&FloatFromCurve(
      FloatCurve=DataInterface() {propertyValues {Curve {keys [
          {arriveTangent=4.0699, interpMode=RCIM_Cubic, leaveTangent=4.0699, tangentMode=RCTM_User, time=0.1, value=0}
          <at 0.5 0.9666 RCIM_Cubic>
          {arriveTangent=-2.5115, interpMode=RCIM_Cubic, leaveTangent=-2.5115, tangentMode=RCTM_User, time=1, value=0}
         ]}}}
      Scale Curve=&RandomRangeFloat(Minimum=0.8, Maximum=1.2)
     )
     B=&FloatFromCurve(
      FloatCurve=DataInterface() {propertyValues {Curve {keys [
          {interpMode=RCIM_Cubic, tangentMode=RCTM_User, time=0, value=0.3}
          {interpMode=RCIM_Cubic, tangentMode=RCTM_User, time=0.1, value=1}
          {interpMode=RCIM_Cubic, tangentMode=RCTM_User, time=0.3, value=1}
          {arriveTangent=-1.6746, interpMode=RCIM_Cubic, leaveTangent=-1.6746, tangentMode=RCTM_User, time=1, value=0.3}
         ]}}}
      CurveIndex=&ReturnNormalizedExecIndex()
     )
    )) {module=ScaleRibbonWidth001}
   &ScaleColor(
    Scale Alpha=&Multiply_Float(
     A=&Multiply_Float(
      A=&Multiply_Float(
       A=&FloatFromCurve(FloatCurve=DataInterface() {propertyValues {Curve {keys [
            {interpMode=RCIM_Cubic, tangentMode=RCTM_Break, time=0.1, value=0}
            {interpMode=RCIM_Cubic, tangentMode=RCTM_Break, time=0.2, value=1}
            {interpMode=RCIM_Cubic, tangentMode=RCTM_Break, time=1, value=0}
           ]}}})
       B=&FloatFromCurve(
        FloatCurve=DataInterface() {propertyValues {Curve {keys [
            {interpMode=RCIM_Cubic, tangentMode=RCTM_User, time=0, value=0}
            {interpMode=RCIM_Cubic, tangentMode=RCTM_User, time=0.2, value=1}
            {interpMode=RCIM_Cubic, tangentMode=RCTM_User, time=0.4, value=1}
            {interpMode=RCIM_Cubic, tangentMode=RCTM_User, time=0.8, value=0}
           ]}}}
        CurveIndex=$User.NormalizedNotifyProgress
       )
      )
      B=&FNTimeOfDayScalar(Night Scale=2, Morning Scale=10, Day Scale=12, Evening Scale=3)
     )
     B=&FloatFromCurve(
      FloatCurve=DataInterface() {propertyValues {Curve {keys [<at 0 0>,<at 0.3 1>,<at 1 0>]}}}
      CurveIndex=&ReturnNormalizedExecIndex()
     )
    )
    ScaleA=true
    ScaleRGB=true
   )
   &SolveForcesAndVelocity(Clamp Velocity=false, Limit Acceleration=false)
  ]
  rendererClasses [&NiagaraRibbonRendererProperties]
  renderers [
   {rendererClass=&NiagaraRibbonRendererProperties}
   {rendererClass=&NiagaraRibbonRendererProperties, rendererIndex=1}
  ]
 }}
summary {
 systemName=NS_Emote_SharpMagnet_SmokeRibbon
 emitters [{emitterName=SmokeRibbon, rendererClasses [&NiagaraRibbonRendererProperties]}]
 userVariables [
  User.bIsFrontEnd {type=NiagaraBool}
  User.bIsFrontEndPreview {type=NiagaraBool}
  User.bIsFullyDeployed {type=NiagaraBool}
  User.ForwardDot {type=NiagaraFloat, defaultValue=0, description=""}
  User.NormalizedNotifyProgress {type=NiagaraFloat, defaultValue=0.4, description=""}
  User.SmokeHeight {type=NiagaraFloat, defaultValue=1, description=""}
 ]
}
